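# Translucent, lit parent material for the sarong on the HammerVice female character, built on the shared glass function with a masked thin-film layer.
# Static switches add camera dither and choose between vertex hiding, which pulls vertices toward the object position, and the standard character opacity function.
# After TechArt overrides, opacity is zeroed where pixel depth exceeds biased custom depth, cutting away sarong pixels behind custom-depth geometry.
# An older view-space depth comparison remains unconnected.
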Material /BRCosmetics/Characters/Player/Female/Medium/Bodies/F_MED_HammerVice/Materials/M_HammerVice_Sarong_Parent
graph {
 properties {
  allowTranslucentCustomDepthWrites=true
  bEnableResponsiveAA=true
  bScreenSpaceReflections=true
  bUseMaterialAttributes=true
  bUsedWithClothing=true
  bUsedWithMeshDeformer=true
  bUsedWithMorphTargets=true
  bUsedWithNanite=true
  bUsedWithNiagaraMeshParticles=true
  bUsedWithSkeletalMesh=true
  blendMode=BLEND_Translucent
  ditherOpacityMask=true
  pixelDepthOffsetMode=PDOM_Legacy
  previewMesh=/CosmeticCompanions/Assets/Quadruped/CosmoHound_Meteor/Meshes/CosmoHound_Meteor_Quadruped
  translucencyDirectionalLightingIntensity=5
  translucencyLightingMode=TLM_SurfacePerPixelLighting
  usedShadingModels=Default Lit
 }
 nodes [
  n1 = Add(
   A=ComponentMask(None=SceneTexture() {sceneTextureId=PPI_CustomDepth}.Color) {r=true}
   B=ScalarParameter() {defaultValue=1, parameterName=CustomDepthBias}
  )
  n2 = Subtract(A=WorldPosition().XYZ, B=CameraPositionWS())
  n3 = Multiply(A=ComponentMask(None=Transform(None=@n2) {transformSourceType=TRANSFORMSOURCE_World, transformType=TRANSFORM_View}) {b=true}) {constB=1}
  n4 = ComponentMask(None=SceneTexture() {sceneTextureId=PPI_SceneDepth}.Color) {r=true}
  Multiply(A=Saturate(None=Subtract(A=@n1, B=@n3)), B=Saturate(None=Subtract(A=@n4, B=@n3)))
  PixelDepth()
  n5 = TextureSampleParameter2D() {
   group=00_Glass
   parameterName=TechArtMask
   samplerType=SAMPLERTYPE_Masks
   sortPriority=-96
   texture=&T_BlackMask
  }
  n6 = MaterialFunctionCall!(
   Material Attributes=MaterialFunctionCall!(TechArtMask=@n5.RGB) {materialFunction=/Game/Global/Materials/Character_Parent_2023/Glass_Parent_2023/MF_Glass}.Result
   Fresnel=Fresnel() {baseReflectFraction=0, exponent=1}
   Alpha=@n5.RGB
  ) {materialFunction=/Game/Global/MaterialFunctions/TechArt/MF_ThinFilm_TechArt}
  n7 = StaticSwitchParameter(
   True=MaterialFunctionCall!(In=@n6.Result) {materialFunction=/Game/Packages/Fortress_SharedMaterials/Base_Material_Functions/BaseCustomizationFunctions/MF_FN_Vehicle_CameraDither}.Result
   False=@n6.Result
  ) {group=00_Glass, parameterName=UseCameraDither, sortPriority=250}
  n8 = ObjectPositionWS()
  n9 = WorldPosition()
  n10 = SetMaterialAttributes!(
   MaterialAttributes=@n7
   World Position Offset=Multiply(A=Subtract(A=@n8, B=@n9.XYZ), B=ScalarParameter() {group=00_Glass, parameterName=VertexHide, sortPriority=301})
  ) {attributeSetTypes [#1]}
  n11 = MaterialFunctionCall!(Material=@n7) {materialFunction=/Game/Global/Materials/Character_Parent_2023/Functions/BaseFN_CharacterOpacity_2023}
  n12 = GetMaterialAttributes(None=MaterialFunctionCall(In=StaticSwitchParameter(True=@n10, False=@n11.Result) {group=00_Glass, parameterName=UseVertexHide, sortPriority=300}) {materialFunction=&MF_TechArtOverrides}.Result) {attributeGetTypes [#2]}
  n13 = PixelDepth()
  n14 = Constant() {r=1}
  n15 = Constant()
  SetMaterialAttributes!(MaterialAttributes=@n12.MaterialAttributes, Opacity=Multiply(A=@n12.Opacity, B=If(A=@n1, B=@n13.R, "A > B"=@n14, "A < B"=@n15))) {attributeSetTypes [#2]}
 ]
}
