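# Physics asset for the F_MED_HammerVice female medium player body cosmetic, covering its dangling bandage bones.
# Two capsule chains, three-segment bandage A and two-segment bandage B, hang from MSTR_bandage through constraints with shared, narrow swing and twist limits.
# Larger capsules on skirt_fot_deform and both thighs have no constraints, apparently serving as collision volumes for the bandages.
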PhysicsAsset /BRCosmetics/Characters/Player/Female/Medium/Bodies/F_MED_HammerVice/Meshes/F_MED_HammerVice_Physics
bodies {
 MSTR_bandage [{shapeType=Capsule, length=1, radius=1, shapeName=MSTR_bandage_capsule, rotation=<rot 90 0 0>}]
 bandage_A_1 [{shapeType=Capsule, length=2.2681, radius=1.6256, shapeName=bandage_A_1_capsule, center=<3.2485 0 0.3>, rotation=<rot 90 0 0>}]
 bandage_A_2 [{shapeType=Capsule, length=2.247, radius=2.0934, shapeName=bandage_A_2_capsule, center=<3 0 0.85>, rotation=<rot 90 0 0>}]
 bandage_A_3 [{shapeType=Capsule, length=1.7225, radius=3.2783, shapeName=bandage_A_3_capsule, center=<3.32 0 0.9>, rotation=<rot 90 0 0>}]
 bandage_B_1 [{shapeType=Capsule, length=1.6263, radius=0.9196, shapeName=bandage_B_1_capsule, center=<2.6085 0 0>, rotation=<rot 90 0 0>}]
 bandage_B_2 [{shapeType=Capsule, length=2.3825, radius=1.5255, shapeName=bandage_B_2_capsule, center=<3.651 0 0>, rotation=<rot 90 0 0>}]
 skirt_fot_deform [{
   shapeType=Capsule
   length=6.367
   radius=12.3019
   shapeName=skirt_fot_deform_base_capsule
   center=<-10.9503 1.0139 -3.8418>
   rotation=<rot 0.0008 -10.1677 89.9999>
  }]
 thigh_l [{shapeType=Capsule, length=14.3786, radius=7.2247, shapeName=thigh_l_capsule, center=<-13.2406 -0.4171 0.2789>, rotation=<rot 90 -4.5672 0>}]
 thigh_r [{shapeType=Capsule, length=17.2037, radius=9.188, shapeName=thigh_r_capsule, center=<14.4014 0 0>, rotation=<rot 90 0 0>}]
}
constraints [
 {bone1Name=bandage_A_3, bone2Name=bandage_A_2, swing1LimitDegrees=25, swing2LimitDegrees=35, twistLimitDegrees=5}
 {bone1Name=bandage_A_2, bone2Name=bandage_A_1, swing1LimitDegrees=25, swing2LimitDegrees=35, twistLimitDegrees=5}
 {bone1Name=bandage_A_1, bone2Name=MSTR_bandage, swing1LimitDegrees=25, swing2LimitDegrees=35, twistLimitDegrees=5}
 {bone1Name=bandage_B_2, bone2Name=bandage_B_1, swing1LimitDegrees=25, swing2LimitDegrees=35, twistLimitDegrees=5}
 {bone1Name=bandage_B_1, bone2Name=MSTR_bandage, swing1LimitDegrees=25, swing2LimitDegrees=35, twistLimitDegrees=5}
]
