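# Post-copy secondary motion AnimBlueprint for the F_MED_HammerVice face accessory mesh, adding physics sway to its hair bones.
# It copies the attached parent's pose, then runs 17 single-bone AnimDynamics nodes in series, hair_A_1 through hair_R.
# Each uses angular springs, damping, gravity and narrow angular limits.
# Shared low/mid/hi variables drive linear acceleration and velocity scales, alongside one shared sim-space settings variable; the EventGraph holds no logic.
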AnimBlueprint /BRCosmetics/Characters/Player/Female/Medium/Bodies/F_MED_HammerVice/Meshes/Parts/F_MED_HammerVice_FaceAcc_AnimBP
target_skeleton=&/F_MED_HammerVice_FaceAcc_Skeleton
graphs [
 AnimGraph {nodes [
   n1 = Get low linear arc scale() -> (low linear arc scale:double)
   n2 = Get low linear vel scale() -> (low linear vel scale:double)
   n3 = Get Sim Space Settings() -> (Sim Space Settings:Anim Phys Sim Space Settings Structure)
   n4 = Get mid linear arc scale() -> (mid linear arc scale:double)
   n5 = Get mid linear vel scale() -> (mid linear vel scale:double)
   n6 = Get Sim Space Settings()
   n7 = Get Sim Space Settings()
   n8 = Get hi linear arc scale() -> (hi linear arc scale:double)
   n9 = Get hi linear vel scale() -> (hi linear vel scale:double)
   n10 = Get Sim Space Settings()
   n11 = Get low linear arc scale()
   n12 = Get low linear vel scale()
   n13 = Anim Dynamics - Bone: hair_A_1!(
    ComponentPose=Local To Component(LocalPose=Copy Pose From Mesh() {bUseAttachedParent=true})
    Alpha=1
    ComponentLinearAccScale_X:double=@n11
    ComponentLinearAccScale_Y:double=@n11
    ComponentLinearAccScale_Z:double=@n11
    ComponentLinearVelScale_X:double=@n12
    ComponentLinearVelScale_Y:double=@n12
    ComponentLinearVelScale_Z:double=@n12
    SimSpaceSettings:Anim Phys Sim Space Settings Structure=@n10
    ExternalForce="0,0,0"
   ) -> (Pose) {
    angularDampingOverride=1
    angularSpringConstant=6000
    bAngularSpring=true
    bOverrideAngularDamping=true
    bUseGravityOverride=true
    boundBone {boneName=hair_A_1}
    componentAppliedLinearAccClamp=<3000 3000 3000>
    componentLinearAccScale=<0.14 0.14 0.14>
    componentLinearVelScale=<0.1 0.1 0.1>
    gravityOverride {z=-980}
    physicsBodyDefinitions [{
      boundBone {boneName=hair_A_1}
      boxExtents=<15 15 15>
      constraintSetup {angularLimitsMax=<0 15 15>, angularLimitsMin=<0 -15 -15>, angularTarget=<1 0 0>}
      localJointOffset=<1 0 0>
     }]
    simSpaceSettings {simSpaceAngularAlpha=0.25}
   }
   n14 = Get low linear arc scale()
   n15 = Get low linear vel scale()
   n16 = Anim Dynamics - Bone: hair_B!(
    ComponentPose=@n13
    Alpha=1
    ComponentLinearAccScale_X:double=@n14
    ComponentLinearAccScale_Y:double=@n14
    ComponentLinearAccScale_Z:double=@n14
    ComponentLinearVelScale_X:double=@n15
    ComponentLinearVelScale_Y:double=@n15
    ComponentLinearVelScale_Z:double=@n15
    SimSpaceSettings:Anim Phys Sim Space Settings Structure=@n10
    ExternalForce="0,0,0"
   ) -> (Pose) {
    angularDampingOverride=1
    angularSpringConstant=6000
    bAngularSpring=true
    bOverrideAngularDamping=true
    bUseGravityOverride=true
    boundBone {boneName=hair_B}
    componentAppliedLinearAccClamp=<3000 3000 3000>
    componentLinearAccScale=<0.35 0.35 0.35>
    componentLinearVelScale=<0.25 0.25 0.25>
    gravityOverride {z=-980}
    physicsBodyDefinitions [{
      boundBone {boneName=hair_B}
      boxExtents=<13 13 13>
      constraintSetup {angularLimitsMax=<0 15 15>, angularLimitsMin=<0 -15 -15>, angularTarget=<1 0 0>}
      localJointOffset=<2.5 0 0>
     }]
   }
   n17 = Get hi linear arc scale()
   n18 = Get hi linear vel scale()
   n19 = Anim Dynamics - Bone: hair_C!(
    ComponentPose=@n16
    Alpha=1
    ComponentLinearAccScale_X:double=@n17
    ComponentLinearAccScale_Y:double=@n17
    ComponentLinearAccScale_Z:double=@n17
    ComponentLinearVelScale_X:double=@n18
    ComponentLinearVelScale_Y:double=@n18
    ComponentLinearVelScale_Z:double=@n18
    SimSpaceSettings:Anim Phys Sim Space Settings Structure=@n10
    ExternalForce="0,0,0"
   ) -> (Pose) {
    angularDampingOverride=1
    angularSpringConstant=6000
    bAngularSpring=true
    bOverrideAngularDamping=true
    bUseGravityOverride=true
    boundBone {boneName=hair_C}
    componentAppliedLinearAccClamp=<3000 3000 3000>
    componentLinearAccScale=<0.42 0.42 0.42>
    componentLinearVelScale=<0.3 0.3 0.3>
    gravityOverride {z=-980}
    physicsBodyDefinitions [{
      boundBone {boneName=hair_C}
      boxExtents=<13 13 13>
      constraintSetup {angularLimitsMax=<0 15 15>, angularLimitsMin=<0 -15 -15>, angularTarget=<1 0 0>}
      localJointOffset=<1.5 0 0>
     }]
   }
   n20 = Get mid linear arc scale()
   n21 = Get mid linear vel scale()
   n22 = Anim Dynamics - Bone: hair_D!(
    ComponentPose=@n19
    Alpha=1
    ComponentLinearAccScale_X:double=@n20
    ComponentLinearAccScale_Y:double=@n20
    ComponentLinearAccScale_Z:double=@n20
    ComponentLinearVelScale_X:double=@n21
    ComponentLinearVelScale_Y:double=@n21
    ComponentLinearVelScale_Z:double=@n21
    SimSpaceSettings:Anim Phys Sim Space Settings Structure=@n10
    ExternalForce="0,0,0"
   ) -> (Pose) {
    angularDampingOverride=1
    angularSpringConstant=6000
    bAngularSpring=true
    bOverrideAngularDamping=true
    bUseGravityOverride=true
    boundBone {boneName=hair_D}
    componentAppliedLinearAccClamp=<3000 3000 3000>
    componentLinearAccScale=<0.245 0.245 0.245>
    componentLinearVelScale=<0.175 0.175 0.175>
    gravityOverride {z=-980}
    physicsBodyDefinitions [{
      boundBone {boneName=hair_D}
      boxExtents=<13 13 13>
      constraintSetup {angularLimitsMax=<0 15 15>, angularLimitsMin=<0 -15 -15>, angularTarget=<1 0 0>}
      localJointOffset=<2.5 0 0>
     }]
   }
   n23 = Get low linear arc scale()
   n24 = Get low linear vel scale()
   n25 = Anim Dynamics - Bone: hair_E!(
    ComponentPose=@n22
    Alpha=1
    ComponentLinearAccScale_X:double=@n23
    ComponentLinearAccScale_Y:double=@n23
    ComponentLinearAccScale_Z:double=@n23
    ComponentLinearVelScale_X:double=@n24
    ComponentLinearVelScale_Y:double=@n24
    ComponentLinearVelScale_Z:double=@n24
    SimSpaceSettings:Anim Phys Sim Space Settings Structure=@n10
    ExternalForce="0,0,0"
   ) -> (Pose) {
    angularDampingOverride=1
    angularSpringConstant=6000
    bAngularSpring=true
    bOverrideAngularDamping=true
    bUseGravityOverride=true
    boundBone {boneName=hair_E}
    componentAppliedLinearAccClamp=<3000 3000 3000>
    componentLinearAccScale=<0.28 0.28 0.28>
    componentLinearVelScale=<0.2 0.2 0.2>
    gravityOverride {z=-980}
    physicsBodyDefinitions [{
      boundBone {boneName=hair_E}
      boxExtents=<13 13 13>
      constraintSetup {angularLimitsMax=<0 15 15>, angularLimitsMin=<0 -15 -15>, angularTarget=<1 0 0>}
      localJointOffset=<2 0 0>
     }]
   }
   n26 = Anim Dynamics - Bone: hair_F!(
    ComponentPose=@n25
    Alpha=1
    ComponentLinearAccScale_X:double=@n8
    ComponentLinearAccScale_Y:double=@n8
    ComponentLinearAccScale_Z:double=@n8
    ComponentLinearVelScale_X:double=@n9
    ComponentLinearVelScale_Y:double=@n9
    ComponentLinearVelScale_Z:double=@n9
    SimSpaceSettings:Anim Phys Sim Space Settings Structure=@n7
    ExternalForce="0,0,0"
   ) -> (Pose) {
    angularDampingOverride=1
    angularSpringConstant=6000
    bAngularSpring=true
    bOverrideAngularDamping=true
    bUseGravityOverride=true
    boundBone {boneName=hair_F}
    componentAppliedLinearAccClamp=<3000 3000 3000>
    componentLinearAccScale=<0.35 0.35 0.35>
    componentLinearVelScale=<0.25 0.25 0.25>
    gravityOverride {z=-980}
    physicsBodyDefinitions [{
      boundBone {boneName=hair_F}
      boxExtents=<13 13 13>
      constraintSetup {angularLimitsMax=<0 15 15>, angularLimitsMin=<0 -15 -15>, angularTarget=<1 0 0>}
      localJointOffset=<2.5 0 0>
     }]
   }
   n27 = Get low linear arc scale()
   n28 = Get low linear vel scale()
   n29 = Anim Dynamics - Bone: hair_G!(
    ComponentPose=@n26
    Alpha=1
    ComponentLinearAccScale_X:double=@n27
    ComponentLinearAccScale_Y:double=@n27
    ComponentLinearAccScale_Z:double=@n27
    ComponentLinearVelScale_X:double=@n28
    ComponentLinearVelScale_Y:double=@n28
    ComponentLinearVelScale_Z:double=@n28
    SimSpaceSettings:Anim Phys Sim Space Settings Structure=@n7
    ExternalForce="0,0,0"
   ) -> (Pose) {
    angularDampingOverride=1
    angularSpringConstant=6000
    bAngularSpring=true
    bOverrideAngularDamping=true
    bUseGravityOverride=true
    boundBone {boneName=hair_G}
    componentAppliedLinearAccClamp=<3000 3000 3000>
    componentLinearAccScale=<0.245 0.245 0.245>
    componentLinearVelScale=<0.175 0.175 0.175>
    gravityOverride {z=-980}
    physicsBodyDefinitions [{
      boundBone {boneName=hair_G}
      boxExtents=<13 13 13>
      constraintSetup {angularLimitsMax=<0 15 15>, angularLimitsMin=<0 -15 -15>, angularTarget=<1 0 0>}
      localJointOffset=<2 0 0>
     }]
   }
   n30 = Get mid linear arc scale()
   n31 = Get mid linear vel scale()
   n32 = Anim Dynamics - Bone: hair_H!(
    ComponentPose=@n29
    Alpha=1
    ComponentLinearAccScale_X:double=@n30
    ComponentLinearAccScale_Y:double=@n30
    ComponentLinearAccScale_Z:double=@n30
    ComponentLinearVelScale_X:double=@n31
    ComponentLinearVelScale_Y:double=@n31
    ComponentLinearVelScale_Z:double=@n31
    SimSpaceSettings:Anim Phys Sim Space Settings Structure=@n7
    ExternalForce="0,0,0"
   ) -> (Pose) {
    angularDampingOverride=1
    angularSpringConstant=6000
    bAngularSpring=true
    bOverrideAngularDamping=true
    bUseGravityOverride=true
    boundBone {boneName=hair_H}
    componentAppliedLinearAccClamp=<3000 3000 3000>
    componentLinearAccScale=<0.35 0.35 0.35>
    componentLinearVelScale=<0.25 0.25 0.25>
    gravityOverride {z=-980}
    physicsBodyDefinitions [{
      boundBone {boneName=hair_H}
      boxExtents=<13 13 13>
      constraintSetup {angularLimitsMax=<0 15 15>, angularLimitsMin=<0 -15 -15>, angularTarget=<1 0 0>}
      localJointOffset=<1.5 0 0>
     }]
   }
   n33 = Get hi linear arc scale()
   n34 = Get hi linear vel scale()
   n35 = Anim Dynamics - Bone: hair_I!(
    ComponentPose=@n32
    Alpha=1
    ComponentLinearAccScale_X:double=@n33
    ComponentLinearAccScale_Y:double=@n33
    ComponentLinearAccScale_Z:double=@n33
    ComponentLinearVelScale_X:double=@n34
    ComponentLinearVelScale_Y:double=@n34
    ComponentLinearVelScale_Z:double=@n34
    SimSpaceSettings:Anim Phys Sim Space Settings Structure=@n7
    ExternalForce="0,0,0"
   ) -> (Pose) {
    angularDampingOverride=1
    angularSpringConstant=6000
    bAngularSpring=true
    bOverrideAngularDamping=true
    bUseGravityOverride=true
    boundBone {boneName=hair_I}
    componentAppliedLinearAccClamp=<3000 3000 3000>
    componentLinearAccScale=<0.42 0.42 0.42>
    componentLinearVelScale=<0.3 0.3 0.3>
    gravityOverride {z=-980}
    physicsBodyDefinitions [{
      boundBone {boneName=hair_I}
      boxExtents=<13 13 13>
      constraintSetup {angularLimitsMax=<0 15 15>, angularLimitsMin=<0 -15 -15>, angularTarget=<1 0 0>}
      localJointOffset=<2.5 0 0>
     }]
   }
   n36 = Get low linear arc scale()
   n37 = Get low linear vel scale()
   n38 = Anim Dynamics - Bone: hair_J!(
    ComponentPose=@n35
    Alpha=1
    ComponentLinearAccScale_X:double=@n36
    ComponentLinearAccScale_Y:double=@n36
    ComponentLinearAccScale_Z:double=@n36
    ComponentLinearVelScale_X:double=@n37
    ComponentLinearVelScale_Y:double=@n37
    ComponentLinearVelScale_Z:double=@n37
    SimSpaceSettings:Anim Phys Sim Space Settings Structure=@n7
    ExternalForce="0,0,0"
   ) -> (Pose) {
    angularDampingOverride=1
    angularSpringConstant=6000
    bAngularSpring=true
    bOverrideAngularDamping=true
    bUseGravityOverride=true
    boundBone {boneName=hair_J}
    componentAppliedLinearAccClamp=<3000 3000 3000>
    componentLinearAccScale=<0.21 0.21 0.21>
    componentLinearVelScale=<0.15 0.15 0.15>
    gravityOverride {z=-980}
    physicsBodyDefinitions [{
      boundBone {boneName=hair_J}
      boxExtents=<13 13 13>
      constraintSetup {angularLimitsMax=<0 15 15>, angularLimitsMin=<0 -15 -15>, angularTarget=<1 0 0>}
      localJointOffset=<1.5 0 0>
     }]
   }
   n39 = Anim Dynamics - Bone: hair_K!(
    ComponentPose=@n38
    Alpha=1
    ComponentLinearAccScale_X:double=@n4
    ComponentLinearAccScale_Y:double=@n4
    ComponentLinearAccScale_Z:double=@n4
    ComponentLinearVelScale_X:double=@n5
    ComponentLinearVelScale_Y:double=@n5
    ComponentLinearVelScale_Z:double=@n5
    SimSpaceSettings:Anim Phys Sim Space Settings Structure=@n6
    ExternalForce="0,0,0"
   ) -> (Pose) {
    angularDampingOverride=1
    angularSpringConstant=6000
    bAngularSpring=true
    bOverrideAngularDamping=true
    bUseGravityOverride=true
    boundBone {boneName=hair_K}
    componentAppliedLinearAccClamp=<3000 3000 3000>
    componentLinearAccScale=<0.35 0.35 0.35>
    componentLinearVelScale=<0.25 0.25 0.25>
    gravityOverride {z=-980}
    physicsBodyDefinitions [{
      boundBone {boneName=hair_K}
      boxExtents=<13 13 13>
      constraintSetup {angularLimitsMax=<0 15 15>, angularLimitsMin=<0 -15 -15>, angularTarget=<1 0 0>}
      localJointOffset=<2 0 0>
     }]
   }
   n40 = Get low linear arc scale()
   n41 = Get low linear vel scale()
   n42 = Anim Dynamics - Bone: hair_L!(
    ComponentPose=@n39
    Alpha=1
    ComponentLinearAccScale_X:double=@n40
    ComponentLinearAccScale_Y:double=@n40
    ComponentLinearAccScale_Z:double=@n40
    ComponentLinearVelScale_X:double=@n41
    ComponentLinearVelScale_Y:double=@n41
    ComponentLinearVelScale_Z:double=@n41
    SimSpaceSettings:Anim Phys Sim Space Settings Structure=@n6
    ExternalForce="0,0,0"
   ) -> (Pose) {
    angularDampingOverride=1
    angularSpringConstant=6000
    bAngularSpring=true
    bOverrideAngularDamping=true
    bUseGravityOverride=true
    boundBone {boneName=hair_L}
    componentAppliedLinearAccClamp=<3000 3000 3000>
    componentLinearAccScale=<0.245 0.245 0.245>
    componentLinearVelScale=<0.175 0.175 0.175>
    gravityOverride {z=-980}
    physicsBodyDefinitions [{
      boundBone {boneName=hair_L}
      boxExtents=<13 13 13>
      constraintSetup {angularLimitsMax=<0 15 15>, angularLimitsMin=<0 -15 -15>, angularTarget=<1 0 0>}
      localJointOffset=<2.5 0 0>
     }]
   }
   n43 = Get hi linear arc scale()
   n44 = Get hi linear vel scale()
   n45 = Anim Dynamics - Bone: hair_M!(
    ComponentPose=@n42
    Alpha=1
    ComponentLinearAccScale_X:double=@n43
    ComponentLinearAccScale_Y:double=@n43
    ComponentLinearAccScale_Z:double=@n43
    ComponentLinearVelScale_X:double=@n44
    ComponentLinearVelScale_Y:double=@n44
    ComponentLinearVelScale_Z:double=@n44
    SimSpaceSettings:Anim Phys Sim Space Settings Structure=@n6
    ExternalForce="0,0,0"
   ) -> (Pose) {
    angularDampingOverride=1
    angularSpringConstant=6000
    bAngularSpring=true
    bOverrideAngularDamping=true
    bUseGravityOverride=true
    boundBone {boneName=hair_M}
    componentAppliedLinearAccClamp=<3000 3000 3000>
    componentLinearAccScale=<0.42 0.42 0.42>
    componentLinearVelScale=<0.3 0.3 0.3>
    gravityOverride {z=-980}
    physicsBodyDefinitions [{
      boundBone {boneName=hair_M}
      boxExtents=<13 13 13>
      constraintSetup {angularLimitsMax=<0 15 15>, angularLimitsMin=<0 -15 -15>, angularTarget=<1 0 0>}
      localJointOffset=<1.5 0 0>
     }]
   }
   n46 = Get low linear arc scale()
   n47 = Get low linear vel scale()
   n48 = Anim Dynamics - Bone: hair_N_1!(
    ComponentPose=@n45
    Alpha=1
    ComponentLinearAccScale_X:double=@n46
    ComponentLinearAccScale_Y:double=@n46
    ComponentLinearAccScale_Z:double=@n46
    ComponentLinearVelScale_X:double=@n47
    ComponentLinearVelScale_Y:double=@n47
    ComponentLinearVelScale_Z:double=@n47
    SimSpaceSettings:Anim Phys Sim Space Settings Structure=@n6
    ExternalForce="0,0,0"
   ) -> (Pose) {
    angularDampingOverride=1
    angularSpringConstant=6000
    bAngularSpring=true
    bOverrideAngularDamping=true
    bUseGravityOverride=true
    boundBone {boneName=hair_N_1}
    componentAppliedLinearAccClamp=<3000 3000 3000>
    componentLinearAccScale=<0.245 0.245 0.245>
    componentLinearVelScale=<0.175 0.175 0.175>
    gravityOverride {z=-980}
    physicsBodyDefinitions [{
      boundBone {boneName=hair_N_1}
      boxExtents=<13 13 13>
      constraintSetup {angularLimitsMax=<0 15 15>, angularLimitsMin=<0 -15 -15>, angularTarget=<1 0 0>}
      localJointOffset=<2.5 0 0>
     }]
   }
   n49 = Get low linear arc scale()
   n50 = Get low linear vel scale()
   n51 = Anim Dynamics - Bone: hair_O!(
    ComponentPose=@n48
    Alpha=1
    ComponentLinearAccScale_X:double=@n49
    ComponentLinearAccScale_Y:double=@n49
    ComponentLinearAccScale_Z:double=@n49
    ComponentLinearVelScale_X:double=@n50
    ComponentLinearVelScale_Y:double=@n50
    ComponentLinearVelScale_Z:double=@n50
    SimSpaceSettings:Anim Phys Sim Space Settings Structure=@n6
    ExternalForce="0,0,0"
   ) -> (Pose) {
    angularDampingOverride=1
    angularSpringConstant=6000
    bAngularSpring=true
    bOverrideAngularDamping=true
    bUseGravityOverride=true
    boundBone {boneName=hair_O}
    componentAppliedLinearAccClamp=<3000 3000 3000>
    componentLinearAccScale=<0.245 0.245 0.245>
    componentLinearVelScale=<0.175 0.175 0.175>
    gravityOverride {z=-980}
    physicsBodyDefinitions [{
      boundBone {boneName=hair_O}
      boxExtents=<13 13 13>
      constraintSetup {angularLimitsMax=<0 15 15>, angularLimitsMin=<0 -15 -15>, angularTarget=<1 0 0>}
      localJointOffset=<2 0 0>
     }]
   }
   n52 = Anim Dynamics - Bone: hair_Q!(
    ComponentPose=@n51
    Alpha=1
    ComponentLinearAccScale_X:double=@n1
    ComponentLinearAccScale_Y:double=@n1
    ComponentLinearAccScale_Z:double=@n1
    ComponentLinearVelScale_X:double=@n2
    ComponentLinearVelScale_Y:double=@n2
    ComponentLinearVelScale_Z:double=@n2
    SimSpaceSettings:Anim Phys Sim Space Settings Structure=@n3
    ExternalForce="0,0,0"
   ) -> (Pose) {
    angularDampingOverride=1
    angularSpringConstant=5000
    bAngularSpring=true
    bOverrideAngularDamping=true
    bUseGravityOverride=true
    boundBone {boneName=hair_Q}
    componentAppliedLinearAccClamp=<3000 3000 3000>
    componentLinearAccScale=<0.245 0.245 0.245>
    componentLinearVelScale=<0.175 0.175 0.175>
    gravityOverride {z=-980}
    physicsBodyDefinitions [{
      boundBone {boneName=hair_Q}
      boxExtents=<14 14 14>
      constraintSetup {angularLimitsMax=<0 0 12>, angularLimitsMin=<0 -10 -10>, angularTarget=<1 0 0>}
      localJointOffset=<2 0 0>
     }]
   }
   n53 = Get low linear arc scale()
   n54 = Get low linear vel scale()
   n55 = Anim Dynamics - Bone: hair_R!(
    ComponentPose=@n52
    Alpha=1
    ComponentLinearAccScale_X:double=@n53
    ComponentLinearAccScale_Y:double=@n53
    ComponentLinearAccScale_Z:double=@n53
    ComponentLinearVelScale_X:double=@n54
    ComponentLinearVelScale_Y:double=@n54
    ComponentLinearVelScale_Z:double=@n54
    SimSpaceSettings:Anim Phys Sim Space Settings Structure=@n3
    ExternalForce="0,0,0"
   ) -> (Pose) {
    angularDampingOverride=1
    angularSpringConstant=5000
    bAngularSpring=true
    bOverrideAngularDamping=true
    bUseGravityOverride=true
    boundBone {boneName=hair_R}
    componentAppliedLinearAccClamp=<3000 3000 3000>
    componentLinearAccScale=<0.245 0.245 0.245>
    componentLinearVelScale=<0.175 0.175 0.175>
    gravityOverride {z=-980}
    physicsBodyDefinitions [{
      boundBone {boneName=hair_R}
      boxExtents=<14 14 14>
      constraintSetup {angularLimitsMax=<0 0 10>, angularLimitsMin=<0 -10 -10>, angularTarget=<1 0 0>}
      localJointOffset=<1.5 0 0>
     }]
   }
   Output Pose(Result=Component To Local(ComponentPose=@n55)) {layerGroup=None}
  ]}
 EventGraph {nodes [Event BlueprintUpdateAnimation()]}
]
variables [
 hi linear arc scale {type=number, default=0.35}
 hi linear vel scale {type=number, default=0.15}
 low linear arc scale {type=number, default=0.1}
 low linear vel scale {type=number, default=0.05}
 mid linear arc scale {type=number, default=0.25}
 mid linear vel scale {type=number, default=0.2}
 Sim Space Settings {
  type=AnimPhysSimSpaceSettings
  default {maxAngularAcceleration=45, maxAngularVelocity=3, simSpaceAngularAlpha=1, externalAngularVelocity=<0 0 0>}
 }
]
